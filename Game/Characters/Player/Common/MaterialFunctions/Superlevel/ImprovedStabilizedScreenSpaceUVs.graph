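# Material function that builds screen-space UVs anchored to an object, so projected patterns stay stable relative to it.
# Each pixel's offset from the pivot plus a local-space offset is projected along the camera ray onto the camera-facing plane through that anchor.
# The result becomes view-space XY, output directly as RawUVs, or Y-flipped, scaled down and shifted as StabilizedUVs.
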MaterialFunction /Game/Characters/Player/Common/MaterialFunctions/Superlevel/ImprovedStabilizedScreenSpaceUVs
graph {
 properties {bExposeToLibrary=true, description="Screen Space UV's stabilzied to Object "}
 nodes [
  n1 = MaterialFunctionCall() {materialFunction=&ObjectPivotPoint}
  n2 = FunctionInput(Preview=Constant3Vector() {constant=<rgba 0 0 110 1>}) {bUsePreviewValueAsDefault=true, description=Local Space offset, inputName=ObjectPositionOffset, sortPriority=1}
  n3 = Subtract(
   A=WorldPosition().XYZ
   B=Add(
    A=FunctionInput(Preview=@n1.Object Pivot Location) {bUsePreviewValueAsDefault=true, description=Object Position, inputName=ObjectPivotLocation}
    B=Transform(None=@n2) {transformSourceType=TRANSFORMSOURCE_Local}
   )
  )
  n4 = MaterialFunctionCall() {materialFunction=&CameraDirectionVector}
  n5 = CameraVectorWS()
  n6 = Multiply(A=Divide(A=DotProduct(A=@n3, B=@n4.Result), B=DotProduct(A=@n4.Result, B=@n5)), B=@n5)
  n7 = ComponentMask(None=Transform(None=Subtract(A=@n3, B=@n6)) {transformSourceType=TRANSFORMSOURCE_World, transformType=TRANSFORM_View}) {g=true, r=true}
  n8 = Multiply(A=@n7, B=Constant2Vector() {g=-1, r=1})
  FunctionOutput(None=Add(A=Divide(A=@n8) {constB=100})) {outputName=StabilizedUVs}
  FunctionOutput(None=@n7) {outputName=RawUVs}
 ]
}
